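# Translucent Niagara material for weapon tracer beams, combining a smoke trail with a glowing colored streak on spline-thickened geometry.
# Two panning smoke samples tile by beam length and erode through dynamic parameters, which also drive the streak's gradient.
# Blended smoke normals dotted with sun/moon direction fake lighting, tinted by sun color and time of day.
# Opacity includes near-camera fade and an optional distance mask around the muzzle.
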Material /Game/Effects/Fort_Effects/Gameplay/Weapons/M_Smoke_Tracer_Beam_Spline
graph {
 properties {
  bUsedWithNiagaraMeshParticles=true
  bUsedWithNiagaraRibbons=true
  bUsedWithNiagaraSprites=true
  blendMode=BLEND_Translucent
  pixelDepthOffsetMode=PDOM_Legacy
  translucencyLightingMode=TLM_VolumetricPerVertexNonDirectional
  twoSided=true
  usedShadingModels=Default Lit
 }
 nodes [
  Constant3Vector() {constant=<rgba 0 0 1 0>}
  Constant() {r=1}
  Constant() {r=1}
  n1 = WorldPosition()
  n2 = DynamicParameter() {parameterIndex=3, paramNames [MuzzlePosX,MuzzlePosY,MuzzlePosZ,MaxDistance]}
  n3 = ParticleColor() {desc="R = erosion, G = panning, B = emissive"}
  n4 = DynamicParameter() {paramNames [Erode, Pan, Gradient 1-0, Length]}
  n5 = ComponentMask(None=TextureCoordinate()) {g=true}
  n6 = OneMinus(None=@n5)
  n7 = Clamp(None=LinearInterpolate(B=Multiply(A=@n4.Length) {constB=50}, Alpha=Multiply(A=@n6, B=@n5)) {constB=40})
  n8 = NamedRerouteUsage()
  n9 = TextureCoordinate() {coordinateIndex=1, vTiling=0.25}
  n10 = Constant2Vector() {g=0.33}
  n11 = TextureCoordinate() {coordinateIndex=1, vTiling=2.5}
  n12 = Constant2Vector() {g=-0.75}
  n13 = ScalarParameter() {defaultValue=1.5, parameterName=SmokeTileDistance}
  n14 = Multiply(
   A=LinearInterpolate(A=Add(A=@n9, B=@n10), B=Add(A=@n11, B=@n12), Alpha=NamedRerouteUsage())
   B=AppendVector(A=Divide(A=@n4.Length, B=@n13), B=Constant() {r=1})
  )
  n15 = Time()
  n16 = Add(
   A=@n4.Pan
   B=Add(A=Multiply(A=@n15) {constB=-0.25}, B=DynamicParameter() {parameterIndex=1, paramNames [WidthMult,Random,Param3,Param4]}.Random)
  )
  n17 = VectorParameter() {parameterName=SmokeSpeed, defaultValue=<rgba -0.1 0 0 1>}
  n18 = Panner(Coordinate=@n14, Time=@n16, Speed=@n17.RGB) {speedX=0.1}
  n19 = TextureSampleParameter2D(UVs=@n18) {parameterName=TrailTexture, samplerType=SAMPLERTYPE_Grayscale, texture=&T_Smoke_Trail_Soft}
  n20 = Multiply(A=Panner(Coordinate=@n14, Time=@n16, Speed=Multiply(A=@n17.RGB) {constB=0.5}) {speedX=0.1}, B=Constant2Vector() {g=1.08, r=1.5}) {constB=2}
  n21 = TextureSampleParameter2D(UVs=@n20) {parameterName=TrailTexture, samplerType=SAMPLERTYPE_Grayscale, texture=&T_Smoke_Trail_Soft}
  n22 = LinearInterpolate(A=@n19.G, B=@n21.G) {constAlpha=0.37}
  n23 = Add(A=@n8, B=@n22)
  n24 = Multiply(A=Multiply(A=LinearInterpolate(Alpha=@n23) {constA=-3.5, constB=1.25}, B=@n8), B=@n22)
  n25 = LinearInterpolate(Alpha=@n8) {constA=0.5, constB=3}
  n26 = Multiply(A=@n7, B=Clamp(None=Power(Base=@n24, Exp=@n25)))
  n27 = Multiply(
   A=Multiply(A=@n3.A, B=Multiply(A=@n7, B=@n26))
   B=MaterialFunctionCall(Distance=ScalarParameter() {defaultValue=400, parameterName=NearCameraFadeDistance}) {materialFunction=&Near_Camera_Fade}.Result
  )
  LinearInterpolate(
   A=@n27
   B=Multiply(A=@n27, B=SphereMask(A=@n1.XYZ, B=@n2.RGB, Radius=@n2.MaxDistance))
   Alpha=ScalarParameter() {parameterName=UseMaxDistance, sliderMax=1}
  ) {constAlpha=0}
  MaterialFunctionCall() {materialFunction=&DebugTimeSine}
  n28 = VectorParameter() {parameterName=SmokeColor, defaultValue=<rgba 0.45 0.45 0.45 0>}
  n29 = Constant() {r=0.55}
  n30 = LinearInterpolate(B=ComponentMask(None=CollectionParameter() {collection=&FortniteMaterialParameters, parameterName=SunLightColor}) {b=true, g=true, r=true}) {constA=0.5}
  n31 = Clamp(None=NamedRerouteUsage(), Min=LinearInterpolate(Alpha=@n22) {constA=0.17, constB=0.08}) {clampMode=CMODE_ClampMin, minDefault=0.1}
  n32 = Add(
   A=MaterialFunctionCall(Color_Morning=@n29, Color_Day=Constant() {r=0.5}, Color_Evening=@n29, Color_Night=Constant() {r=0.7}) {materialFunction=&Color_Time_of_Day}.Result
   B=Multiply(A=Multiply(A=@n30, B=@n31)) {constB=0.5}
  ) {constA=0.5}
  n33 = Multiply(
   A=Multiply(A=Multiply(A=@n28.RGB, B=@n32), B=Multiply(A=@n22, B=@n22))
   B=ScalarParameter() {defaultValue=0.07, parameterName=SmokeBrightness}
  ) {constB=0.07}
  n34 = TextureCoordinate()
  n35 = NamedRerouteUsage()
  n36 = LinearInterpolate(Alpha=@n35) {constA=-1, constB=0.9}
  n37 = Power(
   Base=SphereMask(A=ComponentMask(None=@n34) {g=true}, B=Constant() {r=1}, Radius=Multiply(A=@n35) {constA=1, constB=1.1}, Hardness=Saturate(None=@n36)) {hardnessPercent=5}
   Exp=LinearInterpolate(Alpha=@n35) {constA=3}
  )
  n38 = LinearInterpolate(Alpha=@n35) {constA=-0.1, constB=2}
  n39 = MaterialFunctionCall() {materialFunction=&Swizzle}
  n40 = Multiply(
   A=Clamp(None=Multiply(A=@n37, B=@n38))
   B=Multiply(A=TextureSample(UVs=@n39.YX) {texture=/Game/PvP/StrategicBuildingActors/Defensive/LaserFence/Assets/T_LaserBeamParticle}.G, B=@n7)
  )
  n41 = Multiply(
   A=LinearInterpolate(
    A=VectorParameter() {parameterName=DimColor, defaultValue=<rgba 0.5 0.2602 0.2701 0>}.RGB
    B=VectorParameter() {parameterName=BrightColor, defaultValue=<rgba 1.5 1.5 1.5 0>}.RGB
    Alpha=@n40
   )
   B=@n40
  )
  n42 = Multiply(
   A=MaterialFunctionCall(LDR Input=Clamp(None=@n41), HDR Tint=@n41) {materialFunction=&DeriveHDRfromLDR}.RemappedRange
   B=@n3.RGB
  )
  MaterialFunctionCall(Emissive=Multiply(A=Add(A=@n33, B=@n42))) {materialFunction=&MF_FX_EmissiveAdjustments}
  NamedRerouteDeclaration(None=@n4.Gradient 1-0) {name=GradientParam, nodeColor=<rgba 0.3594 1 0 1>}
  NamedRerouteDeclaration(None=@n4.Erode) {name=ErodeParam, nodeColor=<rgba 0 1 0.6484 1>}
  n43 = Multiply(
   A=ScalarParameter() {defaultValue=50, parameterName=SplineWidth}
   B=DynamicParameter() {parameterIndex=1, paramNames [WidthMult,Random,Param3,Param4]}.WidthMult
  )
  n44 = MaterialFunctionCall(WidthBase=Multiply(A=@n43, B=ScalarParameter() {defaultValue=1, parameterName=WidthEndMult}), WidthTip=@n43) {materialFunction=&SplineThicken}
  n45 = TextureSampleParameter2D(UVs=@n18) {parameterName=TrailNormal, samplerType=SAMPLERTYPE_Normal, texture=/Game/Effects/Fort_Effects/Textures/Smoke/T_Smoke_Trail_Soft_N}
  n46 = TextureSampleParameter2D(UVs=@n20) {parameterName=TrailNormal, samplerType=SAMPLERTYPE_Normal, texture=&T_Smoke_Trail_Soft_N}
  n47 = Transform(None=LinearInterpolate(A=@n45.RGB, B=@n46.RGB) {constAlpha=0.37})
  n48 = CollectionParameter() {
   collection=&FortniteMaterialParameters
   parameterName=SunAndMoonModelDirectionalVector
  }
  NamedRerouteDeclaration(None=DotProduct(
    A=MaterialFunctionCall(BaseNormal=@n44.Normal, AdditionalNormal=@n47) {materialFunction=&BlendAngleCorrectedNormals}.Result
    B=ComponentMask(None=@n48) {b=true, g=true, r=true}
   )) {name=FakeLighting, nodeColor=<rgba 0 0.0156 1 1>}
  OneMinus(None=Time() {bOverride_Period=true, period=1})
  TextureSampleParameter2D() {parameterName=TrailTexture_Normal, texture=/Game/Effects/Fort_Effects/Textures/Smoke/T_Smoke_Trail_Soft_2}
 ]
}
