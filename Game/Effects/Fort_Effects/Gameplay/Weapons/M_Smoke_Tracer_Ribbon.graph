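# Translucent Niagara ribbon material for smoke trails behind weapon tracers.
# Ribbon UVs, randomly stretched per particle and panned by dynamic parameters, feed trail textures whose opacity fades by ribbon start, camera distance, view angle and depth.
# Fake lighting from trail normals, sky atmosphere light direction and skylight color is built alongside a smoke color split between base and emissive.
# Dynamic parameters also drive noise-based world position offset and opacity breakup.
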Material /Game/Effects/Fort_Effects/Gameplay/Weapons/M_Smoke_Tracer_Ribbon
graph {
 properties {bUsedWithNiagaraRibbons=true, blendMode=BLEND_Translucent, pixelDepthOffsetMode=PDOM_Legacy, usedShadingModels=Default Lit}
 nodes [
  Constant()
  Constant() {r=1}
  Constant()
  LinearInterpolate(
   A=ScalarParameter() {defaultValue=-1, parameterName=Tracer Start Offset}
   B=ScalarParameter() {parameterName=Tracer End Offset}
   Alpha=Saturate(None=DynamicParameter() {defaultValue=<rgba 1 1 1 0>, paramNames [Smoke WPO, Smoke Pan, "Random (0-1)", Lerp Noise]}.Smoke WPO)
  ) {constAlpha=1, constB=0}
  n1 = SkyAtmosphereLightDirection()
  LinearInterpolate(A=@n1, B=Constant3Vector() {constant=<rgba 0 0 -1 0>}, Alpha=ScalarParameter() {parameterName=Light Cheat}) {constAlpha=0}
  Multiply()
  n2 = Multiply(A=ParticleColor() {desc="R = erosion, G = panning, B = emissive"}.RGB, B=NamedRerouteUsage())
  Multiply(A=@n2, B=NamedRerouteUsage())
  n3 = TextureCoordinate() {coordinateIndex=1}
  n4 = Abs(None=ConstantBiasScale(None=ComponentMask(None=@n3) {g=true}) {bias=-0.5, scale=2})
  NamedRerouteDeclaration(None=Power(Base=OneMinus(None=@n4), Exp=ScalarParameter() {defaultValue=1, parameterName=Tracer Exponent}) {constExponent=6}) {name=Tracer Shape, nodeColor=<rgba 0.0078 1 0 1>}
  n5 = TextureSampleParameter2D(UVs=NamedRerouteUsage()) {parameterName=TrailNormal, samplerType=SAMPLERTYPE_Normal, texture=/Game/Effects/Fort_Effects/Textures/Smoke/T_Smoke_Trail_Soft_N}
  n6 = VectorParameter() {parameterName=Normal Multiplier, defaultValue=<rgba 1 -1 1 1>}
  n7 = DotProduct(
   A=Normalize(VectorInput=Multiply(A=@n5.RGB, B=@n6.RGB))
   B=Transform(None=@n1) {transformSourceType=TRANSFORMSOURCE_World, transformType=TRANSFORM_Tangent}
  )
  n8 = Add(A=ConstantBiasScale(None=Multiply(A=@n7) {constB=-1})) {constB=0.5}
  n9 = MaterialFunctionCall(Input=@n7, Input Low=Constant() {r=-3}, Input High=Constant() {r=1}) {materialFunction=&RemapValue_0-1}
  n10 = Power(Base=NamedRerouteUsage()) {constExponent=0.5}
  n11 = MaterialFunctionCall() {materialFunction=/Game/Packages/Fortress_SharedMaterials/Base_Material_Functions/SkylightColorAndIntensity.SkyLightColorAndIntensity}
  n12 = ScalarParameter() {defaultValue=1, parameterName=SmokeBrightness}
  NamedRerouteDeclaration(None=Multiply(A=LinearInterpolate(A=@n8, B=@n9.Result, Alpha=@n10) {constA=1}, B=Multiply(A=@n11.Color, B=@n12) {constB=0.07})) {name=FakeLighting, nodeColor=<rgba 0 0.0156 1 1>}
  n13 = ParticleRandom()
  n14 = TextureCoordinate() {coordinateIndex=1}
  n15 = MaterialFunctionCall(Float2=@n14) {materialFunction=&BreakOutFloat2Components}
  n16 = MaterialFunctionCall(X=Multiply(A=LinearInterpolate(Alpha=@n13) {constA=0.85}, B=@n15.R) {constB=1000}, Y=@n15.G) {materialFunction=&MakeFloat2}
  n17 = DynamicParameter() {defaultValue=<rgba 1 1 1 0>, paramNames [Smoke WPO, Smoke Pan, "Random (0-1)", Lerp Noise]}
  n18 = Add(A=@n17.Smoke Pan, B=@n17.Smoke Pan) {constB=0}
  n19 = Constant2Vector() {r=1}
  NamedRerouteDeclaration(None=Add(A=StaticSwitchParameter(True=@n16.Result, False=@n14) {defaultValue=true, parameterName=UseParticleRandom}, B=Multiply(A=@n18, B=@n19))) {name=Ribbon UV by Distance, nodeColor=<rgba 0 0.3906 1 1>}
  n20 = Power(Base=ComponentMask(None=TextureCoordinate()) {r=true}) {constExponent=1.5}
  n21 = NamedRerouteUsage()
  n22 = TextureSample(UVs=Multiply(A=@n21, B=Constant2Vector() {g=1, r=0.75})) {texture=&T_Inky_Smoke_Tile}
  n23 = DynamicParameter() {defaultValue=<rgba 1 1 1 0>, paramNames [Smoke WPO, Smoke Pan, "Random (0-1)", Lerp Noise]}
  n24 = TextureSampleParameter2D(UVs=@n21) {parameterName=TrailTexture, samplerType=SAMPLERTYPE_Grayscale, texture=&T_Smoke_Trail_Soft}
  n25 = Multiply(
   A=ParticleColor() {desc="R = erosion, G = panning, B = emissive"}.A
   B=LinearInterpolate(B=LinearInterpolate(B=@n22.G, Alpha=@n23.Lerp Noise) {constA=1}, Alpha=@n24.G)
  )
  n26 = ComponentMask(None=TextureCoordinate() {coordinateIndex=1}) {r=true}
  n27 = ScalarParameter() {defaultValue=0.05, parameterName=Fade In Distance Multiplier}
  n28 = Multiply(
   A=Saturate(None=Divide(A=@n26, B=@n27))
   B=MaterialFunctionCall(Distance=ScalarParameter() {defaultValue=600, parameterName=NearCameraFadeDistance}) {materialFunction=&Near_Camera_Fade}.Result
  )
  n29 = Multiply(
   A=MaterialFunctionCall("FadeSensitivity (optional)"=ScalarParameter() {defaultValue=0.3, parameterName=FadeBasedOnViewAngleSensitivity}) {materialFunction=&FadeBasedOnViewAngle}.Result
   B=DepthFade(
    Opacity=NamedRerouteDeclaration(None=@n28) {name=Distance Fade, nodeColor=<rgba 0.5 1 0 1>}
    FadeDistance=ScalarParameter() {defaultValue=50, parameterName=Depth Fade Distance}
   )
  )
  NamedRerouteDeclaration(None=Multiply(A=Multiply(A=@n20) {constB=2}, B=Multiply(A=@n25, B=@n29))) {name=Opacity Mask, nodeColor=<rgba 1 0.0703 0 1>}
  n30 = Multiply(A=VectorParameter() {parameterName=SmokeColor, defaultValue=<rgba 0.5 0.5 0.5 1>}.RGB, B=NamedRerouteUsage())
  n31 = ScalarParameter() {defaultValue=0.75, parameterName=Self Emissive}
  NamedRerouteDeclaration(None=Multiply(A=@n30, B=OneMinus(None=@n31))) {name=Base Color, nodeColor=<rgba 1 0 0.2109 1>}
  n32 = Multiply(A=@n30, B=@n31)
  NamedRerouteDeclaration(None=MaterialFunctionCall(Emissive=Add(A=@n32, B=@n2) {constB=0}) {materialFunction=&MF_FX_EmissiveAdjustments}.Result) {name=Emissive Color, nodeColor=<rgba 1 0 0.7969 1>}
  n33 = TextureCoordinate() {coordinateIndex=1}
  n34 = DynamicParameter() {defaultValue=<rgba 1 1 1 0>, paramNames [Smoke WPO, Smoke Pan, "Random (0-1)", Lerp Noise]}
  n35 = TextureSampleParameter2D(UVs=Multiply(A=Add(A=@n33, B=@n34."Random (0-1)"), B=Constant2Vector() {g=0.05, r=0.5})) {parameterName=WPO Noise Texture, samplerType=SAMPLERTYPE_LinearColor, texture=&HighContrastNoise}
  NamedRerouteDeclaration(None=Multiply(A=ConstantBiasScale(None=@n35.RGB) {bias=-0.5, scale=2}, B=@n34.Smoke WPO)) {name=WPO, nodeColor=<rgba 0.3125 1 0 1>}
  NamedRerouteUsage()
  NamedRerouteUsage()
  NamedRerouteUsage()
  NamedRerouteUsage()
  NamedRerouteUsage()
  NamedRerouteUsage()
  ScalarParameter() {defaultValue=1, parameterName=Test}
  n36 = MaterialFunctionCall(Float2=TextureCoordinate()) {materialFunction=&BreakOutFloat2Components}
  n37 = Add(A=@n36.R) {constB=0}
  n38 = Power(Base=OneMinus(None=Multiply(B=@n37) {constA=1}), Exp=ScalarParameter() {defaultValue=1, parameterName=Tracer Exponent}) {constExponent=6}
  SphereMask(
   A=Saturate(None=MaterialFunctionCall(X=@n38, Y=@n36.G) {materialFunction=&MakeFloat2}.Result)
   B=Constant() {r=0.5}
   Radius=ScalarParameter() {defaultValue=0.5, parameterName=Tracer Radius}
   Hardness=ScalarParameter() {parameterName=Tracer Hardness}
  ) {attenuationRadius=0.4, hardnessPercent=0.5}
  WorldPosition() {worldPositionShaderOffset=WPT_ExcludeAllShaderOffsets}
 ]
}
